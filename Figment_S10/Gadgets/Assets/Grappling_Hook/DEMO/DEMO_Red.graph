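# DEMO_Red is a default-lit demo material in the Grappling Hook gadget's asset folder, flagged for skeletal mesh and mesh deformer use.
# Its graph holds four constant color nodes: a saturated red, two greys and a near-white.
# No textures, parameters or math nodes are present, and no connections from the constants to material outputs are shown.
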Material /Figment_S10/Gadgets/Assets/Grappling_Hook/DEMO/DEMO_Red
graph {
 properties {bUsedWithMeshDeformer=true, bUsedWithSkeletalMesh=true, pixelDepthOffsetMode=PDOM_Legacy, usedShadingModels=Default Lit}
 nodes [
  Constant3Vector() {constant=<rgba 0.8073 0.0252 0.0252 1>}
  Constant3Vector() {constant=<rgba 0.2716 0.276 0.2726 1>}
  Constant3Vector() {constant=<rgba 1 1 0.998 1>}
  Constant3Vector() {constant=<rgba 0.3594 0.3594 0.3587 1>}
 ]
}
